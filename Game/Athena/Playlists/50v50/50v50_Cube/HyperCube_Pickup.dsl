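; Pickup Blueprint for the 50v50 Cube playlist's HyperCube item that awards scoring to the player who picks it up.
; On BeginPlay, only the authority binds a custom event to the pickup's OnPickup delegate.
; When picked up, that event casts the interacting pawn's controller to FortPlayerController and adds a scoring event using the ScoreEventTags variable.
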(event EventBeginPlay
  (Networking|SwitchHasAuthority
    (:Authority
      (Interaction|BindEventtoOnPickup (delegate Custom|On%20Pickup%20Event)))))

(event Custom|On%20Pickup%20Event
  (params (SelfActor :type Fort%20Pickup%20Object%20Reference) (InteractingPawn :type Fort%20Pawn%20Object%20Reference))
  (bind _asfort_player_controller (Utilities|Casting|CastToFortPlayerController (Pawn|GetController InteractingPawn)))
  (Scoring|AddScoringEvent _asfort_player_controller (Variables|Default|GetScoreEventTags)))

(fn ConstructionScript)
