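; Damageable weak spot actor for the HyperCube in 50v50, dropping a score pickup when killed.
; Server damage plays a scale pulse with FX and passes the hit to the owning cube.
; A damaging death fires once, leaves the cube's spawned list and moves the cube when none remain.
; One reversible timeline handles grow and shrink; authority-only removal shrinks then destroys it without loot.
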(event Custom|ShrinkWeakSpot
  ;[1 1 1] Grow / Shrink the Weak Spot for Spawn / Destroy
  (bind (_direction _scale) (Timeline|ScaleWeakSpot PlayFromStart)
    (:Update
      (Transformation|SetWorldScale3D (Variables|Default|GetDefaultSceneRoot) (Math|Conversions|ToVector(Float) _scale)))
    (:Finished
      (if (Utilities|Enum|Equal(Enum) _direction "Forward")
        (Actor|DestroyActor)))))

(event EventBeginPlay
  (bind _returnvalue (Transformation|GetActorLocation))
  ;[1 1 1] EVENT :: On Damage (forward to HyperCube for retaliation)
  (Effects|SpawnEmitteratLocation "/Game/Athena/Prototype/Blueprints/Cube/FX/P_C_Bounce_Player.P_C_Bounce_Player" _returnvalue "90.000000, 0, 0") ; Looping beam impact FX
  (Audio|PlaySoundatLocation "/Game/Athena/Sounds/Events/Cube/Cube_LaunchPlayer_Vocal_Cue.Cube_LaunchPlayer_Vocal_Cue" _returnvalue "0, 0, 0" 1.0 1.0 0.0 "/Game/Athena/Sounds/Events/Cube/Cube_AmbientZaps_Att.Cube_AmbientZaps_Att"))

(event Custom|GrowWeakSpot
  ;[1 1 1] Grow / Shrink the Weak Spot for Spawn / Destroy
  (Timeline|ScaleWeakSpot ReverseFromEnd))

(event Custom|DamagePulse
  (bind _returnvalue (Transformation|GetActorLocation))
  ;[1 1 1] FX :: Pulse when taking Damage
  (Development|PrintString "BP_HyperCube_WeakSpot: WeakSpot Damaged!" (Variables|Default|GetDebugPrints) (Variables|Default|GetDebugPrints) "(R=0.585798,G=0.000000,B=1.000000,A=1.000000)")
  (Effects|SpawnEmitteratLocation "/Game/Athena/Prototype/Blueprints/Cube/FX/P_Lightning_Spawn.P_Lightning_Spawn" _returnvalue "90.000000, 0, 0") ; Looping beam impact FX
  (Audio|PlaySoundatLocation "/Game/Athena/Sounds/Events/Cube/Cube_AngryZap_Target_Cue.Cube_AngryZap_Target_Cue" _returnvalue "0, 0, 0" 1.0 1.0 0.0 "/Game/Athena/Sounds/Events/Cube/Cube_AmbientZaps_Att.Cube_AmbientZaps_Att")
  (bind (_direction _scale) (Timeline|OnDamage PlayFromStart)
    (:Update
      (Transformation|SetWorldScale3D (Variables|Default|GetStaticMesh) (Math|Conversions|ToVector(Float) _scale)))))

(event EventOnDamageServer
  (params (Damage :type Float%20%28single-precision%29) (DamageTags :type Gameplay%20Tag%20Container%20Structure%20%28by%20ref%29) (Momentum :type Vector :default "0, 0, 0") (HitInfo :type Hit%20Result%20Structure%20%28by%20ref%29) (InstigatedBy :type Controller%20Object%20Reference) (DamageCauser :type Actor%20Object%20Reference) (EffectContext :type Gameplay%20Effect%20Context%20Handle%20Structure))
  ;[1 1 1] EVENT :: On Damage (forward to HyperCube for retaliation)
  (CallFunction|DamagePulse)
  (bind _ashyper_cube (Utilities|Casting|CastToHyperCube (Actor|GetOwner)))
  (Class|HyperCube|DamagePassThrough _ashyper_cube Damage DamageTags Momentum HitInfo InstigatedBy DamageCauser EffectContext))

(event Custom|RemoveWeakSpot
  ;[1 1 1] CUSTOM EVENT :: Remove WeakSpot (don't spawn loots)
  (Networking|SwitchHasAuthority
    (:Authority
      (CallFunction|DestroyFX))))

(event EventOnDeathServer
  (params (Damage :type Float%20%28single-precision%29) (DamageTags :type Gameplay%20Tag%20Container%20Structure%20%28by%20ref%29) (Momentum :type Vector :default "0, 0, 0") (HitInfo :type Hit%20Result%20Structure%20%28by%20ref%29) (InstigatedBy :type Controller%20Object%20Reference) (DamageCauser :type Actor%20Object%20Reference) (EffectContext :type Gameplay%20Effect%20Context%20Handle%20Structure))
  (bind _returnvalue (Actor|GetOwner))
  ;[1 1 1] EVENT :: On Death
  (Utilities|FlowControl|DoOnce false
    (:Completed
      (if (> Damage 0.0)
        (CallFunction|DeathFX)
        ;[1 1 1] Spawn the Score Pickup
        (bind _returnvalue_2 (Inventory|SpawnPickupInWorldWithClass "/Game/Athena/Playlists/50v50/50v50_Cube/HyperCube_PickupItemDef.HyperCube_PickupItemDef" "/Game/Athena/Playlists/50v50/50v50_Cube/HyperCube_Pickup.HyperCube_Pickup_C" 1 (Transformation|GetWorldLocation (Variables|Default|GetSpawnLocation))))
        (Development|PrintString (Utilities|String|BuildString(Object) "" "BP_HyperCube_WeakSpot: Score Pickup Spawned! 
Object: " _returnvalue_2) (Variables|Default|GetDebugPrints) (Variables|Default|GetDebugPrints) "(R=1.000000,G=0.865573,B=0.000000,A=1.000000)")
        ;[1 1 1] Remove Self from SpawnedWeakSpots and Move Cube when there are none left
        (Utilities|IsValid _returnvalue
          (:Is%20Valid
            (bind _ashyper_cube (Utilities|Casting|CastToHyperCube _returnvalue))
            (Utilities|Array|RemoveItem (Variables|/Game/Athena/Playlists/50v50/50v50_Cube/HyperCube.HyperCube_C|GetSpawnedWeakSpotActors _ashyper_cube) self)
            (if (not (> (Utilities|Array|Length (Variables|/Game/Athena/Playlists/50v50/50v50_Cube/HyperCube.HyperCube_C|GetSpawnedWeakSpotActors _ashyper_cube)) 0))
              (Class|HyperCube|MovetoNextLocation _ashyper_cube))))))))

(event Custom|DeathFX
  (bind _returnvalue (Transformation|GetActorLocation))
  ;[1 1 1] EVENT :: On Death
  (Effects|SpawnEmitteratLocation "/Game/Athena/Prototype/Blueprints/Cube/FX/P_Lightning_Strike_Impact_Medium.P_Lightning_Strike_Impact_Medium" _returnvalue "90.000000, 0, 0") ; Looping beam impact FX
  (Audio|PlaySoundatLocation "/Game/Sounds/Fort_GamePlay_Sounds/Explosives/PropaneTank_Explosion_Cue.PropaneTank_Explosion_Cue" _returnvalue "0, 0, 0" 1.0 1.0 0.0 "/Game/Athena/Sounds/Events/Cube/Cube_AmbientZaps_Att.Cube_AmbientZaps_Att")
  (Development|PrintString (Utilities|String|BuildString(Object) "" "BP_HyperCube_WeaklSpot: WeakSpot KIlled!  
Object: " self) (Variables|Default|GetDebugPrints) (Variables|Default|GetDebugPrints) "(R=0.585798,G=0.000000,B=1.000000,A=1.000000)"))

(event Custom|DestroyFX
  (bind _returnvalue (Transformation|GetActorLocation))
  ;[1 1 1] EVENT :: On Death
  (Effects|SpawnEmitteratLocation "/Game/Athena/Prototype/Blueprints/Cube/FX/P_Lightning_Spawn.P_Lightning_Spawn" _returnvalue "90.000000, 0, 0") ; Looping beam impact FX
  (Audio|PlaySoundatLocation "/Game/Athena/Sounds/Events/Cube/Cube_PickaxeHit_Cue.Cube_PickaxeHit_Cue" _returnvalue "0, 0, 0" 1.0 1.0 0.0 "/Game/Athena/Sounds/Events/Cube/Cube_AmbientZaps_Att.Cube_AmbientZaps_Att")
  (CallFunction|ShrinkWeakSpot)
  (Development|PrintString (Utilities|String|BuildString(Object) "" "BP_HyperCube_WeaklSpot: WeakSpot Removed!  
Object: " self) (Variables|Default|GetDebugPrints) (Variables|Default|GetDebugPrints) "(R=0.585798,G=0.000000,B=1.000000,A=1.000000)"))

(fn ConstructionScript)
