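# Subsurface material for the rift fog in the 50v50 Cube playlist, shading swirling storm clouds with purple lightning glow.
# Two scales of sheared, panning cloud textures drive subsurface color with Fresnel rim, normals and dithered pixel depth offset, with a sphere mask hiding UV pinching.
# Lightning emissive combines view and cloud-depth masks with animated noise and a spawn-time curve.
# Cloud noise displaces vertices along normals; a DepthVersion switch adds outward push and dithered view-based masking.
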Material /Game/Athena/Playlists/50v50/50v50_Cube/M_RiftFog
graph {
 properties {pixelDepthOffsetMode=PDOM_Legacy, shadingModel=MSM_Subsurface, usedShadingModels=Subsurface}
 nodes [
  n1 = TextureCoordinate()
  n2 = Multiply(
   A=MaterialFunctionCall(Float2=@n1) {materialFunction=&BreakOutFloat2Components}.G
   B=ScalarParameter() {defaultValue=1, parameterName=Spin}
  )
  n3 = ScalarParameter() {defaultValue=-0.3, parameterName=Speed}
  n4 = Panner(Coordinate=Add(A=@n1, B=@n2), Speed=@n3) {bFractionalPart=true, speedY=-0.1}
  n5 = ComponentMask(None=TextureCoordinate()) {g=true}
  n6 = Constant()
  n7 = SphereMask(A=@n5, B=@n6, Radius=ScalarParameter() {defaultValue=0.2, parameterName=PinchFix}) {hardnessPercent=10}
  n8 = OneMinus(None=@n7)
  n9 = Multiply(
   A=TextureSample(UVs=@n4) {samplerType=SAMPLERTYPE_LinearGrayscale, texture=/Game/Global/Textures/Default/Tileable_Patterns/T_StormCloudDetail01_M}.R
   B=@n8
  )
  n10 = ScalarParameter() {defaultValue=0.2, parameterName=FineCloudModulation}
  n11 = Panner(Coordinate=Add(A=@n2, B=TextureCoordinate() {uTiling=2, vTiling=2}), Speed=@n3) {bFractionalPart=true, speedY=-0.1}
  n12 = Add(
   A=@n9
   B=LinearInterpolate(
    A=Subtract(B=@n10) {constA=0}
    B=@n10
    Alpha=TextureSample(UVs=@n11) {samplerType=SAMPLERTYPE_LinearGrayscale, texture=&T_StormCloudDetail01_M}.R
   )
  )
  n13 = Multiply(A=VectorParameter() {parameterName=SubSurfaceColor, defaultValue=<rgba 0.0912 0 0.14 1>}.RGB, B=@n12)
  n14 = Fresnel(ExponentIn=ScalarParameter() {defaultValue=0.4337, parameterName=SSS_RimExp}) {baseReflectFraction=0}
  Add(A=@n13, B=Multiply(A=Multiply(A=@n13, B=@n14), B=ScalarParameter() {defaultValue=3, parameterName=SSS_RimIntensity}))
  n15 = OneMinus(None=@n12)
  n16 = MaterialFunctionCall() {materialFunction=&DitherTemporalAA}
  Add(A=Multiply(A=Multiply(A=@n15, B=@n16.Result), B=ScalarParameter() {defaultValue=200, parameterName=PDO}), B=ScalarParameter())
  n17 = MaterialFunctionCall(
   Normal=TextureSample(UVs=@n11) {samplerType=SAMPLERTYPE_Normal, texture=/Game/Global/Textures/Default/Tileable_Patterns/T_StormCloudDetail01_N}.RGB
   Flatness=ScalarParameter() {defaultValue=0.5, parameterName=flatnormal}
  ) {materialFunction=&FlattenNormal}
  n18 = Constant3Vector() {constant=<rgba 1 1 0 0>}
  MaterialFunctionCall(
   Normal=Add(
    A=TextureSample(UVs=@n4) {samplerType=SAMPLERTYPE_Normal, texture=&T_StormCloudDetail01_N}.RGB
    B=Multiply(A=@n17.Result, B=@n18)
   )
   Flatness=@n7
  ) {materialFunction=&FlattenNormal}
  n19 = Subtract(A=Time(), B=ScalarParameter() {defaultValue=1, parameterName=Spawn})
  n20 = CurveAtlasRowParameter(CurveTime=Clamp(None=Divide(A=@n19))) {
   atlas=&AthenaCurveAtlas01
   curve=/Game/Athena/LinearColorCurves/Curve_RiftSpawn
   defaultValue=0.002
   parameterName=Curve01
  }
  n21 = Time()
  n22 = Add(
   A=Multiply(A=@n20.A) {constB=40}
   B=Power(Base=TextureSample(UVs=@n21) {samplerType=SAMPLERTYPE_LinearGrayscale, texture=&T_StormCloudDetail01_M}.R)
  )
  n23 = Constant() {r=0.5}
  n24 = PixelNormalWS()
  n25 = VertexNormalWS()
  n26 = Clamp(None=DotProduct(A=LinearInterpolate(A=@n24, B=@n25), B=CameraVectorWS()))
  n27 = ScalarParameter() {defaultValue=4, parameterName=CloudDepthMaskViewExponent}
  n28 = OneMinus(None=@n12)
  n29 = ScalarParameter() {defaultValue=4, parameterName=CloudDepthMaskContrast}
  n30 = Add(
   A=Power(Base=SphereMask(A=@n5, B=@n6, Radius=@n23) {hardnessPercent=0}, Exp=ScalarParameter() {defaultValue=4, parameterName=LighteningFalloff})
   B=Multiply(A=Power(Base=@n26, Exp=@n27), B=Power(Base=@n28, Exp=@n29))
  )
  Reroute(None=Multiply(A=VectorParameter() {parameterName=LighteningColor, defaultValue=<rgba 12.9435 0 32 1>}.RGB, B=Multiply(A=@n22, B=@n30)))
  ScalarParameter() {defaultValue=1, parameterName=Roughness}
  n31 = ScalarParameter() {defaultValue=0.5, parameterName=MaskExponent}
  n32 = Clamp(None=MaterialFunctionCall(In=Power(Base=@n26, Exp=@n31), Contrast=ScalarParameter() {defaultValue=1, parameterName=Contrast}) {materialFunction=&CheapContrast}.Result)
  n33 = StaticBoolParameter() {parameterName=DepthVersion}
  StaticSwitch(
   True=MaterialFunctionCall(Alpha Threshold=Multiply(A=@n32, B=@n9)) {materialFunction=&DitherTemporalAA}.Result
   False=Constant() {r=1}
   Value=@n33
  )
  n34 = VertexNormalWS()
  n35 = ScalarParameter() {defaultValue=40, parameterName=Push}
  n36 = ScalarParameter() {defaultValue=-40, parameterName=WPO_Intensity02}
  n37 = ScalarParameter() {defaultValue=30, parameterName=WPO_Intensity01}
  n38 = Multiply(A=@n9)
  n39 = Multiply(A=Multiply(A=LinearInterpolate(A=@n36, B=@n37, Alpha=@n38), B=@n8), B=VertexNormalWS())
  StaticSwitch(True=Add(A=@n39, B=Multiply(A=@n34, B=@n35)), False=@n39, Value=@n33)
  VectorParameter() {parameterName=BaseColor, defaultValue=<rgba 0.0104 0.0045 0.0098 1>}
 ]
}
